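# Translucent lit FX material for the LuminousAirplane01 vehicle cosmetic wheel, usable on skeletal meshes, giving a metallic, glowing look over a blurred background.
# A packed RGB texture supplies three spinner masks, and a normal map feeds named reroutes for the Fresnel shading.
# Fresnel rim, logo and outline terms combine with an engine SpiralBlur scene-texture result, driving both base color and emissive.
# Opacity fades the blur function's clamped scene-color output toward a constant through a rerouted input.
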Material /VehicleCosmetics/Wheels/LuminousAirplane01/Materials/M_Wheel_LuminousAirplane01_FX
graph {
 properties {
  bUseMaterialAttributes=true
  bUsedWithMeshDeformer=true
  bUsedWithSkeletalMesh=true
  blendMode=BLEND_Translucent
  pixelDepthOffsetMode=PDOM_Legacy
  translucencyLightingMode=TLM_Surface
  usedShadingModels=Default Lit
 }
 nodes [
  n1 = ScalarParameter() {defaultValue=2.5, parameterName=OutlineIntensity}
  n2 = NamedRerouteUsage()
  n3 = NamedRerouteUsage()
  n4 = Transform(None=NamedRerouteUsage())
  n5 = Add(
   A=Fresnel(ExponentIn=ScalarParameter() {defaultValue=4, parameterName=FresnelValue}, Normal=@n4)
   B=Multiply(A=ScalarParameter() {defaultValue=0.1, parameterName=LogoIntensity}, B=NamedRerouteUsage())
  )
  n6 = NamedRerouteUsage()
  n7 = Clamp(None=Power(Base=LinearInterpolate(A=@n3, B=@n5, Alpha=@n6), Exp=ScalarParameter() {defaultValue=0.9, parameterName=Power}))
  n8 = ScalarParameter() {defaultValue=1, parameterName=Intensity}
  n9 = MaterialFunctionCall!(
   Distance=ScalarParameter() {defaultValue=0.005, parameterName=Blur_Intensity}
   Distance Steps=ScalarParameter() {defaultValue=50, parameterName=Blur_Opacity}
   Radial Steps
   Radial Offset
   TempAA Radial Blur
   TempAA Distance Blur
   Distance Mask
  ) {materialFunction=/Engine/Functions/Engine_MaterialFunctions01/Texturing/SpiralBlur-SceneTexture}
  n10 = Add(
   A=Multiply(A=Constant3Vector() {constant=<rgba 0.9047 0.9047 0.9047 1>}, B=Multiply(A=@n1, B=@n2))
   B=Add(A=Multiply(A=@n7, B=@n8), B=@n9.Result)
  )
  MakeMaterialAttributes!(
   BaseColor=@n10
   Metallic=ScalarParameter() {defaultValue=1, parameterName=Metallic}
   Specular
   Roughness=ScalarParameter() {parameterName=Roughness}
   Anisotropy
   EmissiveColor=@n10
   Opacity=LinearInterpolate(A=@n9.SceneColor clamp to 0, B=Constant(), Alpha=NamedRerouteUsage())
   OpacityMask
   Normal
   Tangent
   WorldPositionOffset
   SubsurfaceColor
   ClearCoat
   ClearCoatRoughness
   AmbientOcclusion
   Refraction
   CustomizedUVs_0
   CustomizedUVs_1
   CustomizedUVs_2
   CustomizedUVs_3
   CustomizedUVs_4
   CustomizedUVs_5
   CustomizedUVs_6
   CustomizedUVs_7
   PixelDepthOffset
   ShadingModel
   Displacement
  )
  n11 = TextureSampleParameter2D() {
   parameterName=Normal_Map
   samplerType=SAMPLERTYPE_Normal
   texture=/VehicleCosmetics/Wheels/LuminousAirplane01/Textures/T_Tire_LuminousAirplane01_FX_N
  }
  NamedRerouteDeclaration(None=@n11.RGB) {name=SpinnerNormal, nodeColor=<rgba 0 0.5312 1 1>}
  n12 = TextureSampleParameter2D() {parameterName=RGB_Map, texture=/VehicleCosmetics/Wheels/LuminousAirplane01/Textures/T_Tire_LuminousAirplane01_FX_RGB}
  NamedRerouteDeclaration(None=@n12.R) {name=SpinnerMaskA, nodeColor=<rgba 0.8516 1 0 1>}
  NamedRerouteDeclaration(None=@n12.G) {name=SpinnerMaskB, nodeColor=<rgba 0.8281 0 1 1>}
  n13 = Fresnel(Normal=@n4) {baseReflectFraction=-0.04, exponent=1.25}
  n14 = SmoothStep(Value=Multiply(A=Subtract(B=@n13) {constA=0}) {constB=4}) {constMax=0.05, constMin=0.04}
  NamedRerouteDeclaration(None=LinearInterpolate(A=Fresnel(Normal=@n4) {baseReflectFraction=0.02, exponent=1.5}, B=Clamp(None=@n14) {maxDefault=0.05}, Alpha=@n14)) {name=FresnelOutput, nodeColor=<rgba 0 1 0.5078 1>}
  NamedRerouteDeclaration(None=@n12.B) {name=SpinnerMaskC, nodeColor=<rgba 0 0.625 1 1>}
 ]
}
